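; Inflatable Cage Saddle ball vehicle (Inflate-a-Bull): the rider rolls and bounces inside it while spending air.
; Suck-in shrinks it and adds gravity; suck-out regrows it, and hits soon after become launches.
; Hits trigger bounce impulses, squash-and-stretch material and effects, plus empty-fuel checks.
; Damage, empty air, deep water or status tags eject the rider, usually applying deployment cooldowns; an optional timer repairs a known scaling bug.
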(event EventOnDeathPlayEffects
  (params (Damage :type Float%20%28single-precision%29) (DamageTags :type Gameplay%20Tag%20Container%20Structure%20%28by%20ref%29) (Momentum :type Vector :default "0, 0, 0") (HitInfo :type Hit%20Result%20Structure%20%28by%20ref%29) (InstigatedBy :type Fort%20Pawn%20Object%20Reference) (DamageCauser :type Actor%20Object%20Reference) (EffectContext :type Gameplay%20Effect%20Context%20Handle%20Structure)))

(event Custom|UseAir
  ;[1 1 1] Use Air over time
  (if (not (CallFunction|CheckforInfiniteLoadedEnergy))
    (GameplayEffects|ApplyGameplayEffectToTarget (Ability|GetFortAbilitySystemComponent (Variables|Actor|GetInstigator)) "/MotherGameplay/Items/CageSaddle/GE_CageSaddle_UseAirOverTime.GE_CageSaddle_UseAirOverTime_C" (Ability|GetFortAbilitySystemComponent (Variables|Actor|GetInstigator)) (Math|Conversions|ToFloat(Integer) (Math|Integer|SelectInt 1 0 (Variables|StateTrackers|GetSuckingIn))) (Class|AthenaFunctionLibrary|MakeSimpleFortEffectContext (Variables|Actor|GetInstigator) (Variables|Actor|GetInstigator)))))

(event Custom|StopInitialGrow)

(event EventBeginPlay
  (bind _instigator (Variables|Actor|GetInstigator))
  (bind _returnvalue (Ability|GetFortAbilitySystemComponent _instigator))
  (bind (_basesphericaldriveparams_radius _basesphericaldriveparams_maxspeedkmh _basesphericaldriveparams_lowspeedaccelerationforce _basesphericaldriveparams_highspeedaccelerationforce _basesphericaldriveparams_maxinclinedeg _basesphericaldriveparams_maxaircontrolforce _basesphericaldriveparams_maxaircontrolspeedkmh _basesphericaldriveparams_shellangulardrag _basesphericaldriveparams_dragcoefficient _basesphericaldriveparams_dragcoefficientautobrake _basesphericaldriveparams_maxautobrakespeedkmh _basesphericaldriveparams_contactrepulsionforce _basesphericaldriveparams_contactthreshold _basesphericaldriveparams_massdirectionmaxangledeg _basesphericaldriveparams_massdirectionstiffness _basesphericaldriveparams_bmassdirectioninvert _basesphericaldriveparams_tractionmultiplier _basesphericaldriveparams_bounceforce) (Utilities|Struct|BreakSphericalDriveParams (Variables|GetBaseSphericalDriveParams_Radius)))
  (bind _speedlinesnscomponent (Variables|CowVehicle|Niagara|GetSpeedLinesNSComponent))
  ;[1 1 1] Begin Play
  (if (Class|AthenaFunctionLibrary|RowtoBool (Variables|Design|GetAutoCorrectForScalingBug))
    (Utilities|Time|SetTimerbyEvent (delegate Custom|CheckForScalingBug) 1.5 true))
  (Fortnite|FortShippingLog (Variables|Actor|GetInstigator) "INFLATEABULL_TEST_LOG")
  (Ability|Tasks|WaitGameplayTagAddToActor (Variables|Actor|GetInstigator) "(TagName=\"Gameplay.Action.Player.CarryingPlayer\")"
    (:then
      (Variables|Default|SetShrinkOnDeath_ (Class|AthenaFunctionLibrary|RowtoBool (Variables|Design|GetShrinkOnDeath)))
      (Variables|Default|SetShrinkWhileSuckingIn_ (Class|AthenaFunctionLibrary|RowtoBool (Variables|Design|GetShrinkWhileSuckingIn)))
      (Variables|Default|SetExitTurnToRight?_ (ScalableFloat|GetValueasBool (Variables|Design|GetExitTurnToRight?)))
      (Utilities|FlowControl|Sequence
        (:then_0
          (Variables|SphericalDriveModel|SetBaseSphericalDriveParams (Utilities|Struct|MakeSphericalDriveParams _basesphericaldriveparams_radius _basesphericaldriveparams_maxspeedkmh _basesphericaldriveparams_lowspeedaccelerationforce _basesphericaldriveparams_highspeedaccelerationforce _basesphericaldriveparams_maxinclinedeg (ScalableFloat|GetValueatLevel (Variables|Design|GetMaxAirControlForce)) (ScalableFloat|GetValueatLevel (Variables|Design|GetMaxAirControlSpeed)) _basesphericaldriveparams_shellangulardrag _basesphericaldriveparams_dragcoefficient _basesphericaldriveparams_dragcoefficientautobrake _basesphericaldriveparams_maxautobrakespeedkmh _basesphericaldriveparams_contactrepulsionforce _basesphericaldriveparams_contactthreshold _basesphericaldriveparams_massdirectionmaxangledeg _basesphericaldriveparams_massdirectionstiffness _basesphericaldriveparams_bmassdirectioninvert _basesphericaldriveparams_tractionmultiplier _basesphericaldriveparams_bounceforce))
          (Transformation|AttachComponentToComponent (Variables|CowVehicle|Audio|GetPhysicsAudioComponent) (Variables|Vehicle|GetSkeletalMesh) "None" "SnapToTarget" "SnapToTarget" "SnapToTarget" true)
          (if (not (CallFunction|CheckforInfiniteLoadedEnergy))
            (GameplayEffects|ApplyGameplayEffectToTarget :self _returnvalue :GameplayEffectClass "/MotherGameplay/Items/CageSaddle/GE_CageSaddle_UseAir_2.GE_CageSaddle_UseAir_2_C" :Target _returnvalue :Level 0.0))
          (GameplayEffects|ApplyGameplayEffectToTarget :self _returnvalue :GameplayEffectClass "/MotherGameplay/Items/CageSaddle/GE_CageSaddle_InVehicle.GE_CageSaddle_InVehicle_C" :Target _returnvalue :Level 0.0)
          (GameplayCue|ExecuteGameplayCue%28Burst%29Local self (Variables|GameplayCues|GetGC_Enter) (Ability|GameplayCue|MakeGameplayCueParameters 0.0))
          (bind _returnvalue_2 (Rendering|Material|CreateDynamicMaterialInstance "/MotherGameplay/Items/CageSaddle/Materials/MI_Gadget_Cage_Saddle1.MI_Gadget_Cage_Saddle1" "None" "None" :!class /Script/Engine.KismetMaterialLibrary))
          (bind _output_get (Variables|CowVehicle|Niagara|SetSquashMID _returnvalue_2))
          (Rendering|Material|SetMaterial (Variables|Vehicle|GetSkeletalMesh) 0 _output_get)
          (Components|Activation|Activate _speedlinesnscomponent false)
          (Niagara|SetNiagaraVariable(Bool) _speedlinesnscomponent "InitialSmokeTransformActive" true)
          (Niagara|SetNiagaraVariable(Bool) _speedlinesnscomponent "BounceSmokeActive" false)
          (Utilities|Time|SetTimerbyEvent (delegate Custom|Spin) 0.02)
          (Networking|SwitchHasAuthority
            (:Authority
              (bind _returnvalue_3 (Utilities|Time|SetTimerbyEvent (delegate Custom|UseAir) 0.5 true))
              (Variables|Default|SetUseAir_TimerHandle _returnvalue_3))))
        (:then_1
          (Ability|Tasks|WaitGameplayTagAddToActor (Variables|Actor|GetInstigator) "(TagName=\"Athena.Gameplay.Status.MarkedForAbduct\")"
            (:Added
              (CallFunction|KickOutPlayer true))))
        (:then_2
          (Ability|Tasks|WaitGameplayTagAddToActor (Variables|Actor|GetInstigator) "(TagName=\"Athena.Gameplay.Status.BeingAbducted\")"
            (:then
              (Ability|Tasks|WaitGameplayTagAddToActor (Variables|Actor|GetInstigator) "(TagName=\"Gameplay.Status.DBNO\")"
                (:then
                  (Game|BindEventtoOnDestroyed (Variables|Actor|GetInstigator) (delegate Custom|PlayerDestroyed)))
                (:Added
                  (CallFunction|KickOutPlayer true))))
            (:Added
              (CallFunction|KickOutPlayer true))))))
    (:Added
      (CallFunction|KickOutPlayer true))) ; Fix for FORT-372686
)

(event Custom|CheckForScalingBug
  (bind _skeletalmesh (Variables|Vehicle|GetSkeletalMesh))
  (bind _returnvalue (Transformation|GetWorldScale _skeletalmesh))
  (bind _returnvalue_2 (Utilities|String|ToString(Vector) _returnvalue))
  (bind _bhidden (Variables|Rendering|GetbHidden))
  (bind _returnvalue_3 (Utilities|String|ToString(Boolean) _bhidden))
  (bind _currentscale (Variables|StateTrackers|GetCurrentScale))
  (bind _returnvalue_4 (Utilities|String|ToString(Float) _currentscale))
  (bind _returnvalue_5 (Transformation|GetActorScale3D))
  (bind _returnvalue_6 (Game|GetPlatformName))
  (bind _returnvalue_7 (Fortnite|Sync|Time|GetServerWorldTimeSeconds))
  (bind _returnvalue_8 (Utilities|String|ToString(Float) _returnvalue_7))
  (bind _normalscale (Variables|Default|GetNormalScale_))
  (bind _returnvalue_9 (Utilities|String|ToString(Float) _normalscale))
  (bind _shrunkscale (Variables|Default|GetShrunkScale_))
  (bind _returnvalue_10 (Utilities|String|ToString(Float) _shrunkscale))
  (bind _lastsuckouttime (Variables|StateTrackers|GetLastSuckOutTime))
  (bind _returnvalue_11 (Utilities|String|ToString(Float) _lastsuckouttime))
  (bind _returnvalue_12 (Utilities|String|ToString(Vector) _returnvalue_5))
  (bind _returnvalue_13 (Utilities|String|Append "INFLATEABULL_SCALINGBUG_HAPPENED " "actor scale" _returnvalue_12 "mesh scale " _returnvalue_2 " hidden " _returnvalue_3 " current scale " _returnvalue_4 " platform " _returnvalue_6 " game time " _returnvalue_8 "normal scale " _returnvalue_9 " small scale " _returnvalue_10 " last suckout time " _returnvalue_11))
  ;[1 1 1] Begin Play
  (if (Math|Boolean|ORBoolean (< (Math|Vector|VectorLength (Transformation|GetActorScale3D)) 0.5) (< (Math|Vector|VectorLength (Transformation|GetWorldScale (Variables|Vehicle|GetSkeletalMesh))) 0.5) (Variables|Rendering|GetbHidden))
    (if (Variables|Default|GetTriedToFixScalingBug)
      (Fortnite|FortShippingLog (Variables|Actor|GetInstigator) (Utilities|String|Append "INFLATEABULL_SCALINGBUG_PERSISTED, kicking out player | " _returnvalue_13))
      (CallFunction|KickOutPlayer false) ; failsafe
      (else
        (Fortnite|FortShippingLog (Variables|Actor|GetInstigator) _returnvalue_13)
        (CallFunction|TickSpawnEffects 1.5 1.5 1.5) ; Fancy scale
        (Rendering|SetActorHiddenInGame false)
        (Variables|Default|SetTriedToFixScalingBug true)))
    (else
      (Variables|Default|SetTriedToFixScalingBug false))))

(event Custom|Spin
  (bind _instigator (Variables|Actor|GetInstigator))
  (bind _returnvalue (Fortnite|GetFortPlayerControllerfromActor _instigator))
  (bind (_returnvalue_roll _returnvalue_pitch _returnvalue_yaw) (Math|Rotator|BreakRotator (Math|Vector|RotatorfromAxisandAngle (Math|Vector|CrossProduct (Math|Vector|Normalize (Vehicle|GetVehicleVelocity)) "0, 0,1.000000") 2e+01)))
  ;[1 1 1] Spin and Launch on Enter
  (Utilities|FlowControl|Sequence
    (:then_1
      (Audio|BroadcastAbilitySpinThrustatLocation 0 self (* (Math|Vector|MakeVector _returnvalue_roll _returnvalue_pitch _returnvalue_yaw) 6.0) 777.0)
      (Utilities|IsValid _returnvalue
        (:Is%20Valid
          (Audio|BroadcastAbilityImpactatLocation 0 self (* (Math|Vector|vector+vector (Math|Vector|Normalize (Math|Vector|ProjectVectorontoPlane (Math|Vector|GetForwardVector (Camera|GetCameraRotation (Variables|/Script/Engine.PlayerController|GetPlayerCameraManager _returnvalue))) "0, 0,1.000000")) (Math|Vector|MakeVector 0.0 0.0 (ScalableFloat|GetValueatLevel (Variables|Design|GetImpulseOnEnterVehicle_VerticalMultiplier)))) (ScalableFloat|GetValueatLevel (Variables|Design|GetImpulseOnEnterVehicle))) 777.0 "None" true))) ; invalid on remote clients
    )))

(event Custom|PlayerDestroyed
  (params (DestroyedActor :type Actor%20Object%20Reference))
  ;[1 1 1] Begin Play
  (Actor|DestroyActor))

(event Custom|SuckInRemoteClients
  (bind _speedlinesnscomponent (Variables|CowVehicle|Niagara|GetSpeedLinesNSComponent))
  ;[1 1 1] Begin Play
  ;[1 1 1] Suck In / Out logic
  (Networking|SwitchHasAuthority
    (:Remote
      (if (not (Pawn|IsLocallyControlled (Variables|Actor|GetInstigator)))
        (CallFunction|StartShinking)
        (Effects|Components|ParticleSystem|SetBoolParameter _speedlinesnscomponent "SuckingIn" true)
        (Niagara|SetNiagaraVariable(Float) _speedlinesnscomponent "BounceCharge" 1.0)
        (Variables|StateTrackers|SetSuckingIn true)
        (GameplayEffects|ApplyGameplayEffectToTarget :self (Ability|GetFortAbilitySystemComponent self) :GameplayEffectClass "/MotherGameplay/Items/CageSaddle/GE_IncreaseGravity.GE_IncreaseGravity_C" :Target (Ability|GetFortAbilitySystemComponent self) :Level 0.0)
        (Vehicle|OverridePontoonRadiusonServer 0 (* (Vehicle|GetPontoonRadius 0) (Variables|Default|GetPontoonRadiusScalar)))))))

(event EventOnPawnEnterVehicle
  (params (PlayerPawn :type Fort%20Player%20Pawn%20Object%20Reference) (SeatIdx :type Integer))
  ;[1 1 1] Begin Play
  ;[0 0 0] IsHiddingInProp is dangrous!!
  ; Doing this on enter and exit instead of BeginPlay and KickOut because players were able to go invisible if they entered and exited quickly.
  (bind _asplayer_pawn_athena (Utilities|Casting|CastToPlayerPawn_Athena PlayerPawn))
  (Class|PlayerPawnAthena|IsHiddinginProp _asplayer_pawn_athena true))

(event Custom|SuckIn
  (bind _speedlinesnscomponent (Variables|CowVehicle|Niagara|GetSpeedLinesNSComponent))
  ;[1 1 1] Begin Play
  ;[1 1 1] Suck In / Out logic
  (CallFunction|SuckinRemoteClients)
  (CallFunction|StartShinking)
  (Effects|Components|ParticleSystem|SetBoolParameter _speedlinesnscomponent "SuckingIn" true)
  (Niagara|SetNiagaraVariable(Float) _speedlinesnscomponent "BounceCharge" 1.0)
  (Variables|StateTrackers|SetSuckingIn true)
  (GameplayEffects|ApplyGameplayEffectToTarget :self (Ability|GetFortAbilitySystemComponent self) :GameplayEffectClass "/MotherGameplay/Items/CageSaddle/GE_IncreaseGravity.GE_IncreaseGravity_C" :Target (Ability|GetFortAbilitySystemComponent self) :Level 0.0)
  (Vehicle|OverridePontoonRadiusonServer 0 (* (Vehicle|GetPontoonRadius 0) (Variables|Default|GetPontoonRadiusScalar))))

(event Custom|SuckOut
  (bind _grow_cachedstartingscale (Variables|Default|GetGrow_CachedStartingScale))
  (bind _speedlinesnscomponent (Variables|CowVehicle|Niagara|GetSpeedLinesNSComponent))
  (bind _normalscale (Variables|Default|GetNormalScale_))
  ;[1 1 1] Suck In / Out logic
  (CallFunction|SuckOutRemoteClients)
  (CallFunction|StopShrinking)
  (Components|Activation|Activate _speedlinesnscomponent false)
  (Effects|Components|ParticleSystem|SetBoolParameter _speedlinesnscomponent "SuckingIn" false)
  (Niagara|SetNiagaraVariable(Float) _speedlinesnscomponent "BounceCharge" 0.0)
  (Variables|StateTrackers|SetSuckingIn false)
  (GameplayEffects|RemoveActiveEffectswithGrantedTags (Ability|GetFortAbilitySystemComponent self) "(GameplayTags=((TagName=\"Granted.Athena.CageSaddle.GravityIncrease\")))")
  (Vehicle|OverridePontoonRadiusonServer 0 (/ (Vehicle|GetPontoonRadius 0) (Variables|Default|GetPontoonRadiusScalar)))
  (Variables|StateTrackers|SetLastSuckOutTime (Fortnite|Sync|Time|GetServerWorldTimeSeconds))
  (Variables|Default|SetGrow_CachedStartingScale (Variables|StateTrackers|GetCurrentScale))
  (if (Variables|Default|GetShrinkWhileSuckingIn_)
    (bind (_direction _scale) (Timeline|Grow_ PlayFromStart)
      (:Update
        (CallFunction|TickSpawnEffects (Utilities|ClampScale (Math|Float|Lerp _grow_cachedstartingscale _normalscale _scale)) (Utilities|ClampScale (Math|Float|Lerp _grow_cachedstartingscale _normalscale _scale)) (Utilities|ClampScale (Math|Float|Lerp _grow_cachedstartingscale _normalscale _scale))) ; Fancy scale
        (Variables|StateTrackers|SetCurrentScale (Math|Float|Lerp _grow_cachedstartingscale _normalscale _scale))))))

(event EventOnPawnExitVehicle
  (params (PlayerPawn :type Fort%20Player%20Pawn%20Object%20Reference) (ExitSocketName :type Name :default "None"))
  ;[0 0 0] IsHiddingInProp is dangrous!!
  ; Doing this on enter and exit instead of BeginPlay and KickOut because players were able to go invisible if they entered and exited quickly.
  (Transformation|SetWorldScale3D (Variables|/Script/Engine.Character|GetCapsuleComponent PlayerPawn) "1.000000,1.000000,1.000000")
  (bind _asplayer_pawn_athena (Utilities|Casting|CastToPlayerPawn_Athena PlayerPawn))
  (Class|PlayerPawnAthena|IsHiddinginProp _asplayer_pawn_athena false))

(event Custom|SuckOutRemoteClients
  (bind _grow_cachedstartingscale (Variables|Default|GetGrow_CachedStartingScale))
  (bind _speedlinesnscomponent (Variables|CowVehicle|Niagara|GetSpeedLinesNSComponent))
  (bind _normalscale (Variables|Default|GetNormalScale_))
  ;[1 1 1] Suck In / Out logic
  (Networking|SwitchHasAuthority
    (:Remote
      (if (not (Pawn|IsLocallyControlled (Variables|Actor|GetInstigator)))
        (CallFunction|StopShrinking)
        (Components|Activation|Activate _speedlinesnscomponent false)
        (Effects|Components|ParticleSystem|SetBoolParameter _speedlinesnscomponent "SuckingIn" false)
        (Niagara|SetNiagaraVariable(Float) _speedlinesnscomponent "BounceCharge" 0.0)
        (Variables|StateTrackers|SetSuckingIn false)
        (GameplayEffects|RemoveActiveEffectswithGrantedTags (Ability|GetFortAbilitySystemComponent self) "(GameplayTags=((TagName=\"Granted.Athena.CageSaddle.GravityIncrease\")))")
        (Vehicle|OverridePontoonRadiusonServer 0 (/ (Vehicle|GetPontoonRadius 0) (Variables|Default|GetPontoonRadiusScalar)))
        (Variables|StateTrackers|SetLastSuckOutTime (Fortnite|Sync|Time|GetServerWorldTimeSeconds))
        (Variables|Default|SetGrow_CachedStartingScale (Variables|StateTrackers|GetCurrentScale))
        (if (Variables|Default|GetShrinkWhileSuckingIn_)
          (Timeline|Grow_ PlayFromStart))))))

(event Custom|ExitGC_RemoteClients
  ;[1 1 1] Kick Out Player
  (if (not (Pawn|IsLocallyControlled (Variables|Actor|GetInstigator)))
    (GameplayCue|ExecuteGameplayCue%28Burst%29Local (Vehicle|GetDriver) (select (Variables|StateTrackers|GetDamaged) (Variables|GameplayCues|GetGC_Popped) (Variables|GameplayCues|GetGC_Exit)) (Ability|GameplayCue|MakeGameplayCueParameters 0.0))))

(event Custom|StartShinking
  (bind _normalscale (Variables|Default|GetNormalScale_))
  (bind _shrunkscale (Variables|Default|GetShrunkScale_))
  (CallFunction|StopInitialGrow)
  (if (Variables|Default|GetShrinkWhileSuckingIn_)
    (bind (_direction _scale) (Timeline|Shrink PlayFromStart)
      (:Update
        (Variables|StateTrackers|SetCurrentScale (Math|Float|Lerp _normalscale _shrunkscale _scale))
        (CallFunction|TickSpawnEffects (Utilities|ClampScale (Math|Float|Lerp _normalscale _shrunkscale _scale)) (Utilities|ClampScale (Math|Float|Lerp _normalscale _shrunkscale _scale)) (Utilities|ClampScale (Math|Float|Lerp _normalscale _shrunkscale _scale))) ; Fancy scale
      ))))

(event Custom|StopShrinking
  (bind _normalscale (Variables|Default|GetNormalScale_))
  (bind _shrunkscale (Variables|Default|GetShrunkScale_))
  (Timeline|Shrink Stop))

(event EventOnDamaged
  (params (Damage :type Float%20%28single-precision%29) (DamageTags :type Gameplay%20Tag%20Container%20Structure%20%28by%20ref%29) (EffectContext :type Gameplay%20Effect%20Context%20Handle%20Structure%20%28by%20ref%29) (EventInstigator :type Controller%20Object%20Reference) (DamageCauser :type Actor%20Object%20Reference))
  ;[1 1 1] If in Creative, only send OnDamaged to clients if damage is more than 0. FORT-396123. Otherwise, treat normally.
  (if (Fortnite|IsCreativeorPlaygroundModePlaylist)
    (if (> Damage 0.0)
      _))
  (CallFunction|MulticastonDamaged EventInstigator))

(event Custom|PlayGC
  (params (LastNormalForce :type Vector) (HitLocation :type Vector))
  (bind _returnvalue (Math|Vector|VectorLength LastNormalForce))
  (bind _returnvalue_2 (> _returnvalue 12.0))
  ;[1 1 1] Bounce FX
  (if (and (or (< (- (Variables|StateTrackers|GetLastHitTime) (Variables|StateTrackers|GetLastSuckOutTime)) (Variables|Design|GetTimeWindowForLaunch)) _returnvalue_2) (> (- (Fortnite|Sync|Time|GetServerWorldTimeSeconds) (Variables|Default|GetLastGCTime)) 0.4))
    (Variables|Default|SetLastGCTime (Fortnite|Sync|Time|GetServerWorldTimeSeconds))
    (GameplayCue|ExecuteGameplayCue%28Burst%29Local self (Variables|GameplayCues|GetLaunchCue) (Ability|GameplayCue|MakeGameplayCueParameters :NormalizedMagnitude 1.0 :RawMagnitude (Math|Vector|VectorLength LastNormalForce) :Location HitLocation :Normal "0,0,0" :GameplayEffectLevel 1 :AbilityLevel 1 :bReplicateLocationWhenUsingMinimalRepProxy false))
    (CallFunction|PlayBounceAudio HitLocation (select _returnvalue_2 LastNormalForce "0, 0,60.000000"))))

(event Custom|KickOutPlayer
  (params (ForcedOut :type Boolean))
  (bind _instigator (Variables|Actor|GetInstigator))
  (bind _returnvalue (Ability|GetFortAbilitySystemComponent _instigator))
  ;[1 1 1] Kick Out Player
  (bind _returnvalue_2 (GameplayTags|HasMatchingGameplayTag(Message) (Variables|Actor|GetInstigator) "(TagName=\"Gameplay.Action.Player.AthenaInVehicle\")"))
  (if (and _returnvalue_2 (> (Actor|GetGameTimeSinceCreation) 0.5))
    (Utilities|FlowControl|DoOnce false
      (:Completed
        (GameplayEffects|RemoveActiveEffectswithGrantedTags (Ability|GetFortAbilitySystemComponent (Variables|Actor|GetInstigator)) "(GameplayTags=((TagName=\"Granted.Athena.CageSaddle.InVehicle\")))")
        (bind _output_get (Variables|Default|SetForced%20Out ForcedOut))
        (if (not _output_get)
          (CallFunction|StopShrinking)
          (if (Variables|Default|GetShrinkOnDeath_)
            (bind (_direction _shrink _rotationacceleration) (Timeline|ShrinkAndDestroy PlayFromStart)
              (:Update
                (CallFunction|TickSpawnEffects (Math|Float|Clamp(Float) (Class|AthenaFunctionLibrary|UnclampedLerp 1.0 (Variables|StateTrackers|GetCurrentScale) _shrink) 1.0 3.0) (Math|Float|Clamp(Float) (Class|AthenaFunctionLibrary|UnclampedLerp 1.0 (Variables|StateTrackers|GetCurrentScale) _shrink) 1.0 3.0) (Math|Float|Clamp(Float) (Class|AthenaFunctionLibrary|UnclampedLerp 1.0 (Variables|StateTrackers|GetCurrentScale) _shrink) 1.0 3.0)) ; Fancy scale
                (if (Variables|Default|GetExitTurnToRight?_)
                  (bind _returnvalue_3 (FortPhysicsSimulationLibrary|ComputeTorqueforTargetRotation self (Utilities|GetClosestRotationThatWorkswithShrink (Transformation|GetActorRotation self)) 2e+01 5.0 (Math|Float|Lerp 0.0 1e+03 _rotationacceleration)))
                  (Audio|BroadcastAbilitySpinThrustatLocation 0 self (+ (Physics|GetPhysicsAngularVelocityinDegrees (Variables|Vehicle|GetSkeletalMesh) "None") (* (- _returnvalue_3 (Physics|GetPhysicsAngularVelocityinDegrees (Variables|Vehicle|GetSkeletalMesh) "None")) (* (Utilities|Time|GetWorldDeltaSeconds) 2.0))) 777.0)))
              (:NewTrack_2
                _))))
        (Utilities|FlowControl|DoOnce false
          (:Completed
            (if (Variables|Default|GetRanOutOfAir)
              _
              (elif (Inventory|GetHasNoCooldownsfromContext _instigator)
                _
                (elif (Variables|Default|GetForced%20Out)
                  (GameplayEffects|ApplyGameplayEffectToTarget :self _returnvalue :GameplayEffectClass "/MotherGameplay/Items/CageSaddle/GE_BlockDeployment_CageSaddle.GE_BlockDeployment_CageSaddle_C" :Target _returnvalue :Level 1.0)
                  (GameplayEffects|ApplyGameplayEffectToTarget :self _returnvalue :GameplayEffectClass "/MotherGameplay/Items/CageSaddle/GE_Repairing_CageSaddle.GE_Repairing_CageSaddle_C" :Target _returnvalue :Level 1.0)
                  (GameplayEffects|ApplyGameplayEffectToTarget :self _returnvalue :GameplayEffectClass "/MotherGameplay/Items/CageSaddle/GE_TimerTicking_CageSaddle.GE_TimerTicking_CageSaddle_C" :Target _returnvalue :Level 1.0)
                  (else
                    (GameplayEffects|ApplyGameplayEffectToTarget :self _returnvalue :GameplayEffectClass "/MotherGameplay/Items/CageSaddle/GE_BlockDeployment_CageSaddle.GE_BlockDeployment_CageSaddle_C" :Target _returnvalue :Level 0.0)
                    (GameplayEffects|ApplyGameplayEffectToTarget :self _returnvalue :GameplayEffectClass "/MotherGameplay/Items/CageSaddle/GE_TimerTicking_CageSaddle.GE_TimerTicking_CageSaddle_C" :Target _returnvalue :Level 0.0)))))
            (Utilities|FlowControl|Sequence
              (:then_0
                (bind _asfort_player_controller_gameplay (Utilities|Casting|CastToFortPlayerControllerGameplay (Pawn|GetController (Vehicle|GetDriver))))
                (Class|FortPlayerControllerGameplay|ExitSpawnedVehicle _asfort_player_controller_gameplay))
              (:then_1
                (Networking|SwitchHasAuthority
                  (:Authority
                    (CallFunction|ExitGCRemoteClients))
                  (:Remote
                    (GameplayCue|ExecuteGameplayCue%28Burst%29Local (Vehicle|GetDriver) (select (Variables|StateTrackers|GetDamaged) (Variables|GameplayCues|GetGC_Popped) (Variables|GameplayCues|GetGC_Exit)) (Ability|GameplayCue|MakeGameplayCueParameters 0.0))))))))))
    (else
      (Utilities|FlowControl|DelayUntilNextTick)
      (CallFunction|KickOutPlayer ForcedOut))))

(event EventOnDamagePlayEffects
  (params (Damage :type Float%20%28single-precision%29) (DamageTags :type Gameplay%20Tag%20Container%20Structure%20%28by%20ref%29) (Momentum :type Vector :default "0, 0, 0") (HitInfo :type Hit%20Result%20Structure%20%28by%20ref%29) (InstigatedBy :type Fort%20Pawn%20Object%20Reference) (DamageCauser :type Actor%20Object%20Reference) (EffectContext :type Gameplay%20Effect%20Context%20Handle%20Structure)))

(event Custom|CheckForFuelGone
  (bind _instigator (Variables|Actor|GetInstigator))
  (bind _returnvalue (Ability|GetFortAbilitySystemComponent _instigator))
  (bind (_bsuccessfullyfoundattribute _returnvalue_2) (Ability|Attribute|GetFloatAttribute (Variables|Actor|GetInstigator) "(AttributeName=\"Fuel\",Attribute=/Script/FortniteGame.FortCowVehicleSet:Fuel,AttributeOwner=None)"))
  ;[1 1 1] Kick out on hit if out of fuel
  (if (not (CallFunction|CheckforInfiniteLoadedEnergy))
    (if (and _bsuccessfullyfoundattribute (<= _returnvalue_2 0.0))
      (Variables|Default|SetRanOutOfAir true)
      (CallFunction|KickOutPlayer true)
      (bind _asfort_player_controller_athena (Utilities|Casting|CastToFortPlayerControllerAthena (Fortnite|GetFortPlayerControllerfromActor (Variables|Actor|GetInstigator)))
        (:then
          (if (not (Class|FortPlayerControllerAthena|HasInfiniteConsumables _asfort_player_controller_athena))
            _
            (elif (Inventory|GetHasNoCooldownsfromContext _instigator)
              (Utilities|Time|ClearandInvalidateTimerbyHandle (Variables|Default|GetUseAir_TimerHandle))
              (GameplayEffects|ApplyGameplayEffectToTarget :self _returnvalue :GameplayEffectClass "/MotherGameplay/Items/CageSaddle/GE_CageSaddle_RefillAir.GE_CageSaddle_RefillAir_C" :Target _returnvalue :Level 0.0)
              (else
                (GameplayEffects|ApplyGameplayEffectToTarget :self _returnvalue :GameplayEffectClass "/MotherGameplay/Items/CageSaddle/GE_BlockDeployment_CageSaddle.GE_BlockDeployment_CageSaddle_C" :Target _returnvalue :Level 1.0)
                (GameplayEffects|ApplyGameplayEffectToTarget :self _returnvalue :GameplayEffectClass "/MotherGameplay/Items/CageSaddle/GE_Repairing_CageSaddle.GE_Repairing_CageSaddle_C" :Target _returnvalue :Level 1.0)
                (GameplayEffects|ApplyGameplayEffectToTarget :self _returnvalue :GameplayEffectClass "/MotherGameplay/Items/CageSaddle/GE_TimerTickingBeforeRefill_CageSaddle.GE_TimerTickingBeforeRefill_CageSaddle_C" :Target _returnvalue :Level 1.0)))))
        (:CastFailed
          _))
      (bind _asplayer_pawn_athena (Utilities|Casting|CastToPlayerPawn_Athena (Game|GetInstigator)))
      (Fortnite|DestroyIteminQuickBarSlot _asplayer_pawn_athena "/MotherGameplay/Items/CageSaddle/AGID_CageSaddle_V2.AGID_CageSaddle_V2")
      (bind _asfort_player_pawn (Utilities|Casting|CastToFortPlayerPawn (Variables|Actor|GetInstigator)))
      (FailedAction|UnabletoPerformAction _asfort_player_pawn false (GameplayTags|MakeLiteralGameplayTagContainer "(GameplayTags=((TagName=\"Athena.Ability.Failed.CageSaddle.RanOutOfAir\")))")))))

(event Custom|MulticastOnDamaged
  (params (Instigator :type Controller%20Object%20Reference))
  ;[1 1 1] Kick Out on damage
  (if (or (not (Team|OnSameTeam Instigator (Variables|Actor|GetInstigator))) (Utilities|Equal(Object) (Fortnite|GetFortPlayerControllerfromActor (Variables|Actor|GetInstigator)) (Utilities|Casting|CastToFortPlayerController Instigator)))
    (bind _returnvalue (GameplayTags|HasMatchingGameplayTag(Message) (Variables|Actor|GetInstigator) "(TagName=\"Athena.Gameplay.Status.BeingAbducted\")"))
    (if (not _returnvalue)
      (Variables|StateTrackers|SetDamaged true)
      (CallFunction|KickOutPlayer true)
      (bind _asfort_player_pawn (Utilities|Casting|CastToFortPlayerPawn (Variables|Actor|GetInstigator)))
      (FailedAction|UnabletoPerformAction _asfort_player_pawn false (GameplayTags|MakeGameplayTagContainerfromTag "(TagName=\"Athena.Ability.Failed.CageSaddle.StillRepairing\")")))))

(event Custom|Bounce
  (params (Magnitude :type Float%20%28double-precision%29) (Min :type Float%20%28double-precision%29) (IsLaunch :type Boolean) (Force :type Boolean))
  (bind _lastnormalforce (Variables|Cached|GetLastNormalForce))
  (bind _returnvalue (Math|Vector|VectorLength _lastnormalforce))
  (bind _returnvalue_2 (> _returnvalue 5.0))
  (bind _lastnormalforce_2 (Variables|Cached|GetLastNormalForce))
  (bind _returnvalue_3 (Math|Vector|SelectVector _lastnormalforce_2 "0, 0,15.000000" _returnvalue_2))
  (bind _lastnormalforce_3 (Variables|Cached|GetLastNormalForce))
  (bind _returnvalue_4 (Math|Vector|Normalize _lastnormalforce_3))
  (bind _returnvalue_5 (Transformation|GetVelocity))
  (bind _returnvalue_6 (Math|Vector|ProjectVectorontoPlane _returnvalue_5 _returnvalue_4))
  (bind _maxbouncevel (Variables|Design|GetMaxBounceVel))
  (bind _returnvalue_7 (ScalableFloat|GetValueatLevel _maxbouncevel))
  (bind _returnvalue_8 (* _returnvalue_3 Magnitude))
  (bind _outputpin (Utilities|ClampXY _returnvalue_8))
  (bind _returnvalue_9 (Math|Vector|ClampVectorSize _outputpin Min _returnvalue_7))
  (bind _outputpin_2 (Utilities|ClampXY _returnvalue_6))
  (bind _returnvalue_10 (+ _returnvalue_9 _outputpin_2))
  ;[1 1 1] Bounce Logic
  (if (or (> (- (Fortnite|Sync|Time|GetServerWorldTimeSeconds) (Variables|StateTrackers|GetLastBouncedTime)) 0.25) Force)
    (if (Utilities|CheckifAlreadyFaster _returnvalue_10)
      (Variables|StateTrackers|SetLastBouncedTime (Fortnite|Sync|Time|GetServerWorldTimeSeconds))
      (Audio|BroadcastAbilityImpactatLocation 0 self _returnvalue_10 777.0))))

(event Custom|PlayGCMulticast
  (params (LastNormalForce :type Vector) (HitLocation :type Vector))
  ;[1 1 1] Bounce Logic
  (CallFunction|PlayGC LastNormalForce HitLocation))

(event Custom|PlayGC_AllClients
  ;[1 1 1] Bounce Logic
  (Networking|SwitchHasAuthority
    (:Authority
      (CallFunction|PlayGCMulticast (Variables|Cached|GetLastNormalForce) (Variables|Cached|GetHitLocation)))
    (:Remote
      (CallFunction|PlayGC (Variables|Cached|GetLastNormalForce) (Variables|Cached|GetHitLocation)))))

(event EventOnHit
  (params (Location :type Vector%20%28by%20ref%29 :default "0, 0, 0") (Normal :type Vector%20%28by%20ref%29 :default "0, 0, 0") (Force :type Float%20%28single-precision%29) (HitActor :type Actor%20Object%20Reference))
  (bind _lastsuckouttime (Variables|StateTrackers|GetLastSuckOutTime))
  (bind _lasthittime (Variables|StateTrackers|GetLastHitTime))
  (bind _returnvalue (- _lasthittime _lastsuckouttime))
  (bind _timewindowforlaunch (Variables|Design|GetTimeWindowForLaunch))
  (bind _returnvalue_2 (< _returnvalue _timewindowforlaunch))
  (bind _allowrollingupwalls (Variables|Design|GetAllowRollingUpWalls))
  (bind _returnvalue_3 (Class|AthenaFunctionLibrary|RowtoBool _allowrollingupwalls))
  (bind _returnvalue_4 (Fortnite|Sync|Time|GetServerWorldTimeSeconds))
  (bind _lastbouncedtime (Variables|StateTrackers|GetLastBouncedTime))
  (bind _returnvalue_5 (- _returnvalue_4 _lastbouncedtime))
  (bind _returnvalue_6 (> _returnvalue_5 1.0))
  (bind _returnvalue_7 (or _returnvalue_6 _returnvalue_3))
  (bind _returnvalue_8 (and _returnvalue_7 _returnvalue_2))
  (bind _speedlinesnscomponent (Variables|CowVehicle|Niagara|GetSpeedLinesNSComponent))
  ;[1 1 1] BounceOnHit
  (Utilities|Casting|CastToFortLauncherAthena HitActor
    (:CastFailed
      (Utilities|FlowControl|Sequence
        (:then_0
          (CallFunction|CheckforFuelGone))
        (:then_1
          (bind _output_get (Variables|Cached|SetHitLocation Location))
          (Effects|Components|ParticleSystem|SetVectorParameter (Variables|CowVehicle|Niagara|GetSpeedLinesNSComponent) "SmokeTrailLocation" _output_get)
          (CallFunction|SquashandStretchDetectImpact Force (Variables|Cached|GetHitLocation))
          (Utilities|Casting|CastToPlayerPawn_Athena HitActor
            (:CastFailed
              (Niagara|SetNiagaraVariable(Bool) _speedlinesnscomponent "InitialSmokeTransformActive" false)
              (Niagara|SetNiagaraVariable(Bool) _speedlinesnscomponent "BounceSmokeActive" true)
              (Components|Activation|Activate _speedlinesnscomponent false)
              (Niagara|SetNiagaraVariable(Vector3) _speedlinesnscomponent "SmokePosition" (Variables|Cached|GetHitLocation))
              (Niagara|SetNiagaraVariable(Float) _speedlinesnscomponent "ForceStrength" Force)
              (bind _output_get_2 (Variables|StateTrackers|SetIsLaunching _returnvalue_2))
              (Effects|Components|ParticleSystem|SetBoolParameter _speedlinesnscomponent "IsLaunching" _output_get_2)
              (Variables|Cached|SetLastNormalForce (* Normal Force))
              (Variables|StateTrackers|SetLastHitTime _returnvalue_4)
              (CallFunction|PlayGCAllClients)
              (bind _asfort_vehicle_interface (Utilities|Casting|CastToFortVehicleInterface HitActor)
                (:then
                  (Utilities|Casting|CastToCageSaddleVehicle HitActor
                    (:CastFailed
                      (if (> (* (Math|Vector|DotProduct Normal (Math|Vector|Normalize (Vehicle|GetVehicleVelocity _asfort_vehicle_interface))) (Math|Vector|VectorLength (Vehicle|GetVehicleVelocity _asfort_vehicle_interface))) 1.5e+02)
                        (Variables|Cached|SetLastNormalForce (* Normal (* (Math|Vector|DotProduct Normal (Math|Vector|Normalize (Vehicle|GetVehicleVelocity _asfort_vehicle_interface))) (Math|Vector|VectorLength (Vehicle|GetVehicleVelocity _asfort_vehicle_interface)))))
                        (CallFunction|Bounce (ScalableFloat|GetValueatLevel (Variables|Design|GetBounceMagnitudeFromVehicle))) ; VehicleImpulse
                      ))) ; fix for FORT-382624
                )
                (:CastFailed
                  (if (or (> Force (ScalableFloat|GetValueatLevel (Variables|Design|GetMinForceToBounce))) _returnvalue_8)
                    (CallFunction|Bounce (ScalableFloat|GetValueatLevel (Variables|Design|GetBounceMagnitude)) (Math|Float|SelectFloat (ScalableFloat|GetValueatLevel (Variables|Design|GetLaunchVelFromGround)) 0.0 _returnvalue_8) _returnvalue_8)))))) ; causes desync if bounced by player
        )))))

(event Custom|CheckForAirRunOut
  (CallFunction|CheckforFuelGone))

(event EventOnEnteredWaterBody
  (params (WaterBody :type Fort%20Water%20Body%20Actor%20Object%20Reference))
  ;[1 1 1] Water
  (bind _output_get (Variables|Default|SetWater%20Body WaterBody))
  (Variables|Default|SetEnteringWaterBody true)
  (GameplayCue|ExecuteGameplayCue%28Burst%29Local self (Variables|GameplayCues|GetGC_EnterWater) (Ability|GameplayCue|MakeGameplayCueParameters :NormalizedMagnitude 0.0 :RawMagnitude 0.0 :Location "0,0,0" :Normal "0,0,0" :SourceObject _output_get :GameplayEffectLevel 1 :AbilityLevel 1 :bReplicateLocationWhenUsingMinimalRepProxy false))
  (Projectile|AddGameplayCue%28Looping%29Local self (Variables|GameplayCues|GetGC_InWater) (Ability|GameplayCue|MakeGameplayCueParameters :NormalizedMagnitude 0.0 :RawMagnitude 0.0 :Location "0,0,0" :Normal "0,0,0" :Instigator self :GameplayEffectLevel 1 :AbilityLevel 1 :bReplicateLocationWhenUsingMinimalRepProxy false))
  (CallFunction|CheckforFuelGone)
  (if (not (Utilities|Time|IsValidTimerHandle (Variables|Default|GetCheckFuelTimer)))
    (bind _returnvalue (Utilities|Time|SetTimerbyEvent (delegate Custom|CheckForAirRunOut) 1.0 true))
    (Variables|Default|SetCheckFuelTimer _returnvalue))
  (Utilities|FlowControl|DoOnce false :!label _label
    (:Completed
      (Audio|SpawnSoundAttached (select (Variables|Default|GetEnteringWaterBody) "/MotherGameplay/Sounds/CageSaddle/Cage_Saddle_Water_Splash_Enter_Cue.Cage_Saddle_Water_Splash_Enter_Cue" "/MotherGameplay/Sounds/CageSaddle/Cage_Saddle_Water_Splash_Exit_Cue.Cage_Saddle_Water_Splash_Exit_Cue") (Variables|Vehicle|GetSkeletalMesh))
      (Utilities|FlowControl|Delay 1.0)
      (goto _label Reset))))

(event EventOnExitedWaterBody
  (params (WaterBody :type Fort%20Water%20Body%20Actor%20Object%20Reference))
  ;[1 1 1] Water
  (Variables|Default|SetEnteringWaterBody false)
  (Projectile|RemoveGameplayCue%28Looping%29Local self (Variables|GameplayCues|GetGC_InWater) (Ability|GameplayCue|MakeGameplayCueParameters 0.0))
  (Utilities|Time|ClearandInvalidateTimerbyHandle (Variables|Default|GetCheckFuelTimer))
  (Utilities|FlowControl|DoOnce false :!label _label
    (:Completed
      (Audio|SpawnSoundAttached (select (Variables|Default|GetEnteringWaterBody) "/MotherGameplay/Sounds/CageSaddle/Cage_Saddle_Water_Splash_Enter_Cue.Cage_Saddle_Water_Splash_Enter_Cue" "/MotherGameplay/Sounds/CageSaddle/Cage_Saddle_Water_Splash_Exit_Cue.Cage_Saddle_Water_Splash_Exit_Cue") (Variables|Vehicle|GetSkeletalMesh))
      (Utilities|FlowControl|Delay 1.0)
      (goto _label Reset))))

(event OnComponentBeginOverlap(PlayerCollision)
  (params (OverlappedComponent :type Primitive%20Component%20Object%20Reference) (OtherActor :type Actor%20Object%20Reference) (OtherComp :type Primitive%20Component%20Object%20Reference) (OtherBodyIndex :type Integer) (bFromSweep :type Boolean) (SweepResult :type Hit%20Result%20Structure%20%28by%20ref%29))
  (Utilities|Casting|CastToFortLauncherAthena OtherActor)
  (Audio|BroadcastAbilityImpactatLocation 0 self (Math|Vector|MakeVector 0.0 0.0 5e+03) 777.0))

(event EventOnWaterTooDeep
  (CallFunction|KickOutPlayer true))

(fn CanInteract
  (params (InController :type Controller%20Object%20Reference))
  (return false))

(fn CheckForInfiniteLoadedEnergy
  (return (ScalableFloat|GetValueasBoolwithContextParam (Variables|Actor|GetInstigator) (Variables|/Script/FortniteGame.FortExperienceSettingsComponent|GetInfiniteLoadedEnergy (Fortnite|Athena|GameState|GetExperienceSettingsComponentfromGameState)))))

(fn PlayBounceAudio
  (params (HitLocation :type Vector%20%28by%20ref%29) (HitNormalImpulse :type Vector%20%28by%20ref%29) (Is%20Launch :type Boolean))
  (bind _returnvalue (Vehicle|ClientIsDriver))
  (bind _fortcollisionaudio (Variables|Default|GetFortCollisionAudio))
  (Networking|CanExecuteCosmeticEvents
    (:True
      (Audio|SetCollisionSoundatIndex _fortcollisionaudio 0 (select _returnvalue (Variables|Sound|GetSoundOnBounce1P) (Variables|Sound|GetSoundOnBounce3P)))
      (Audio|SetCollisionSoundatIndex _fortcollisionaudio 1 (select _returnvalue (Variables|Sound|GetSoundOnJump1P) (Variables|Sound|GetSoundOnJump3P)))
      (Audio|OnCollision _fortcollisionaudio HitLocation (Math|Vector|SelectVector (* HitNormalImpulse 6e+01) HitNormalImpulse Is%20Launch)))))

(fn SquashAndStretchDetectImpact
  (params (Force :type Float%20%28double-precision%29) (ForceLocation :type Vector))
  (bind _returnvalue (Math|Float|Absolute(Float) Force))
  (bind _squashmid (Variables|CowVehicle|Niagara|GetSquashMID))
  ;[1 1 1] Squash and Stretch Based on Force
  (if (> _returnvalue 15.0)
    (Variables|CowVehicle|Niagara|SetbIsSquashing true)
    (Variables|CowVehicle|Niagara|SetSquashTime 0.0)
    (bind _output_get (Variables|CowVehicle|Niagara|ImpactInterp|SetForceNormalized (Math|Float|MapRangeClamped _returnvalue 0.0 8e+01 0.0 1.0)))
    (Variables|CowVehicle|Niagara|SetTargetSquash (Math|Float|MapRangeClamped _output_get 0.0 1.0 0.25 1.0))
    (Variables|CowVehicle|Niagara|SetCurrentSquashSpeed (Math|Float|MapRangeClamped _output_get 0.0 1.0 3e+01 1e+01))
    (Rendering|Material|SetVectorParameterValue _squashmid "ImpactPointWS" (Math|Conversions|ToLinearColor(Vector) ForceLocation) :!class /Script/Engine.MaterialInstanceDynamic)
    (bind _output_get_2 (Variables|CowVehicle|Niagara|ImpactInterp|SetImpactPointInterp 1.0))
    (Rendering|Material|SetScalarParameterValue _squashmid "UseWorldImpactPoint" _output_get_2 :!class /Script/Engine.MaterialInstanceDynamic)))

(fn ConstructionScript
  (Variables|FortAthenaVehicle|SetOverlapComponent 0))
